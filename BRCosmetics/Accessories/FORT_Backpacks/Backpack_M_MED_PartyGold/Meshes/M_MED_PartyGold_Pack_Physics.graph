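# Physics asset for the PartyGold backpack cosmetic, providing collision bodies and joint limits for its dangling tassel.
# Capsules cover the pack, left clavicle, left upper arm, sword and each tassel bone, with two capsules each on the pack and tassel tip.
# Constraints chain the four tassel bones from dyn_tassel_MSTR with limited swing and tight twist.
# The pack, arm and sword bodies have no constraints.
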PhysicsAsset /BRCosmetics/Accessories/FORT_Backpacks/Backpack_M_MED_PartyGold/Meshes/M_MED_PartyGold_Pack_Physics
bodies {
 Party_Gold_backpack_clavicle_l [{shapeType=Capsule, length=9, radius=9.5, center=<4.2446 -1.6077 -0.5393>, rotation=<rot 69.3297 1.4476 176.9462>}]
 Party_Gold_backpack_upperarm_l [{
   shapeType=Capsule
   length=10
   radius=7
   shapeName=Party_Gold_backpack_upperarm_l_capsule
   center=<6.4766 -0.5308 -1.2693>
   rotation=<rot -89.3895 22.9581 -26.2542>
  }]
 dyn_sword [{shapeType=Capsule, length=28.1, radius=1.9, shapeName=dyn_sword_capsule, center=<-12.4003 0.2897 0>, rotation=<rot 90 23.4081 32.8707>}]
 dyn_tassel_01 [{shapeType=Capsule, length=1, radius=1, shapeName=dyn_tassel_01_capsule, center=<0.5 0 0>, rotation=<rot 90 0 0>}]
 dyn_tassel_02 [{shapeType=Capsule, length=1, radius=1, shapeName=dyn_tassel_02_capsule, center=<0.5 0 0>, rotation=<rot 90 0 0>}]
 dyn_tassel_03 [{shapeType=Capsule, length=1, radius=1, shapeName=dyn_tassel_03_capsule, center=<0.5 0 0>, rotation=<rot 90 0 0>}]
 dyn_tassel_04 [
  {shapeType=Capsule, length=1, radius=1, shapeName=dyn_tassel_04_capsule, center=<0.5 0.5 0>, rotation=<rot 90 0 0>}
  {
   shapeType=Capsule
   length=5.8
   radius=1.5
   shapeName=dyn_tassel_04_capsule
   center=<3.2154 0.7782 -0.9813>
   rotation=<rot 36.5568 -4.1993 -29.6751>
  }
 ]
 dyn_tassel_MSTR [{shapeType=Capsule, length=0.5, radius=0.5, shapeName=dyn_tassel_MSTR_capsule}]
 pack [
  {shapeType=Capsule, length=14, radius=8, shapeName=pack_capsule, center=<0 -5.3254 3.1087>, rotation=<rot 90 0 0>}
  {shapeType=Capsule, length=6.5, radius=9.5, shapeName=pack_capsule, center=<0 -3.0824 -9.6271>, rotation=<rot 90 0 0>}
 ]
}
constraints [
 {bone1Name=dyn_tassel_01, bone2Name=dyn_tassel_MSTR, swing1LimitDegrees=60, swing2LimitDegrees=30, twistLimitDegrees=10}
 {bone1Name=dyn_tassel_02, bone2Name=dyn_tassel_01, swing1LimitDegrees=50, swing2LimitDegrees=50, twistLimitDegrees=10}
 {bone1Name=dyn_tassel_03, bone2Name=dyn_tassel_02, swing1LimitDegrees=50, swing2LimitDegrees=50, twistLimitDegrees=10}
 {bone1Name=dyn_tassel_04, bone2Name=dyn_tassel_03, swing1LimitDegrees=50, swing2LimitDegrees=50, twistLimitDegrees=10}
]
