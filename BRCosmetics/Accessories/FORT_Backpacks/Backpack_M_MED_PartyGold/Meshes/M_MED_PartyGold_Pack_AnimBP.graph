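# Anim Blueprint for the PartyGold backpack cosmetic, adding secondary motion to the pack.
# A cached start pose nudges the pack in the front end and world-space replaces its left clavicle and upper-arm bones.
# bRigidBodyEnabled selects between RigidBody simulation and a four-bone tassel AnimDynamics chain with spherical limits around pack and sword.
# Each update stores a gravity override from a deprecated helper and the body mesh's clavicle and upper-arm transforms.
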AnimBlueprint /BRCosmetics/Accessories/FORT_Backpacks/Backpack_M_MED_PartyGold/Meshes/M_MED_PartyGold_Pack_AnimBP
target_skeleton=&/M_MED_PartyGold_Pack_Skeleton
graphs [
 AnimGraph {nodes [
   n1 = RigidBody!(ComponentPose=Local To Component(LocalPose=Use cached pose 'Start Pose'()), Alpha, OverrideWorldGravity:Vector="0,0,0", ExternalForce) -> (Pose) {
    bOverrideWorldGravity=true
    overridePhysicsAsset=&/M_MED_PartyGold_Pack_Physics
    alphaBoolBlend {blendOption=Linear}
    componentAppliedLinearAccClamp=<700 700 700>
    componentLinearAccScale=<0.4 0.4 0.4>
    componentLinearVelScale=<0.5 0.5 0.5>
   }
   n2 = Anim Dynamics - Bone: dyn_tassel_01!(ComponentPose=Local To Component(LocalPose=Use cached pose 'Start Pose'()), Alpha=1, GravityOverride:Vector="0,0,0", ExternalForce="0,0,0") -> (Pose) {
    angularDampingOverride=0.999
    bOverrideAngularDamping=true
    bUseGravityOverride=true
    bUseSphericalLimits=true
    alphaBoolBlend {blendOption=Linear}
    boundBone {boneName=dyn_tassel_01}
    componentAppliedLinearAccClamp=<550 550 550>
    componentLinearAccScale=<0.4 0.4 0.4>
    componentLinearVelScale=<0.5 0.5 0.5>
    physicsBodyDefinitions [{
      collisionType=CustomSphere
      sphereCollisionRadius=1.2
      boundBone {boneName=dyn_tassel_01}
      boxExtents=<3 8 8>
      constraintSetup {angularLimitsMax=<5 20 0>, angularLimitsMin=<-5 -60 -30>, angularTarget=<0 0 0>}
      localJointOffset=<0.2 0 0>
     }]
    sphericalLimits [
     {limitRadius=8.3, drivingBone {boneName=pack}, sphereLocalOffset=<8.6695 -4.9247 4.7748>}
     {limitRadius=2.5, drivingBone {boneName=dyn_sword}, sphereLocalOffset=<-18.2489 1.3648 -0.178>}
    ]
   }
   n3 = Anim Dynamics - Bone: dyn_tassel_02!(ComponentPose=@n2, Alpha=1, GravityOverride:Vector="0,0,0", ExternalForce="0,0,0") -> (Pose) {
    angularDampingOverride=0.999
    bOverrideAngularDamping=true
    bUseGravityOverride=true
    bUseSphericalLimits=true
    alphaBoolBlend {blendOption=Linear}
    boundBone {boneName=dyn_tassel_02}
    componentAppliedLinearAccClamp=<550 550 550>
    componentLinearAccScale=<0.4 0.4 0.4>
    componentLinearVelScale=<0.5 0.5 0.5>
    physicsBodyDefinitions [{
      collisionType=CustomSphere
      sphereCollisionRadius=1.2
      boundBone {boneName=dyn_tassel_02}
      boxExtents=<3 7 7>
      constraintSetup {angularLimitsMax=<5 40 0>, angularLimitsMin=<-5 -40 -50>, angularTarget=<0 0 0>}
      localJointOffset=<0.4 0 0>
     }]
    sphericalLimits [
     {limitRadius=8.3, drivingBone {boneName=pack}, sphereLocalOffset=<8.6695 -4.9247 4.7748>}
     {limitRadius=2.5, drivingBone {boneName=dyn_sword}, sphereLocalOffset=<-18.2489 1.3648 -0.178>}
    ]
   }
   n4 = Anim Dynamics - Bone: dyn_tassel_03!(ComponentPose=@n3, Alpha=1, GravityOverride:Vector="0,0,0", ExternalForce="0,0,0") -> (Pose) {
    angularDampingOverride=0.999
    bOverrideAngularDamping=true
    bUseGravityOverride=true
    bUseSphericalLimits=true
    alphaBoolBlend {blendOption=Linear}
    boundBone {boneName=dyn_tassel_03}
    componentAppliedLinearAccClamp=<550 550 550>
    componentLinearAccScale=<0.4 0.4 0.4>
    componentLinearVelScale=<0.5 0.5 0.5>
    physicsBodyDefinitions [{
      collisionType=CustomSphere
      sphereCollisionRadius=1.2
      boundBone {boneName=dyn_tassel_03}
      boxExtents=<3 6 6>
      constraintSetup {angularLimitsMax=<5 40 0>, angularLimitsMin=<-5 -40 -50>, angularTarget=<0 0 0>}
      localJointOffset=<0.5 0 0>
     }]
    sphericalLimits [
     {limitRadius=8.3, drivingBone {boneName=pack}, sphereLocalOffset=<8.6695 -4.9247 4.7748>}
     {limitRadius=2.5, drivingBone {boneName=dyn_sword}, sphereLocalOffset=<-18.2489 1.3648 -0.178>}
    ]
   }
   n5 = Anim Dynamics - Bone: dyn_tassel_04!(ComponentPose=@n4, Alpha=1, GravityOverride:Vector="0,0,0", ExternalForce="0,0,0") -> (Pose) {
    angularDampingOverride=0.999
    bOverrideAngularDamping=true
    bUseGravityOverride=true
    bUseSphericalLimits=true
    alphaBoolBlend {blendOption=Linear}
    boundBone {boneName=dyn_tassel_04}
    componentAppliedLinearAccClamp=<550 550 550>
    componentLinearAccScale=<0.4 0.4 0.4>
    componentLinearVelScale=<0.5 0.5 0.5>
    physicsBodyDefinitions [{
      collisionType=CustomSphere
      sphereCollisionRadius=2.5
      boundBone {boneName=dyn_tassel_04}
      boxExtents=<3 6 6>
      constraintSetup {angularLimitsMax=<5 40 0>, angularLimitsMin=<-5 -40 -50>, angularTarget=<0 0 0>}
      localJointOffset=<0.5 0 0>
     }]
    sphericalLimits [
     {limitRadius=8.3, drivingBone {boneName=pack}, sphereLocalOffset=<8.6695 -4.9247 4.7748>}
     {limitRadius=2.5, drivingBone {boneName=dyn_sword}, sphereLocalOffset=<-18.2489 1.3648 -0.178>}
     {limitRadius=14, drivingBone {boneName=pack}, sphereLocalOffset=<3.2459 8.4414 5.5953>}
    ]
   }
   Output Pose(Result=Blend Poses by bool(BlendPose_0=Component To Local(ComponentPose=@n1), BlendPose_1=Component To Local(ComponentPose=@n5), bActiveValue=Get bRigidBodyEnabled()) {blendType=Linear}) {layerGroup=None}
   n6 = "Transform (Modify) Bone - Bone: pack"!(
    ComponentPose=Local To Component(LocalPose=Input Pose())
    bAlphaBoolEnabled:bool=Get bIsInFrontEnd()
    Translation="0,1,0"
    Rotation="0,0,0"
    Scale
   ) -> (Pose) {
    alphaInputType=Bool
    translationMode=BMM_Additive
    translationSpace=BCS_BoneSpace
    alphaBoolBlend {blendOption=Linear}
    boneToModify {boneName=pack}
    translation {y=1}
   }
   n7 = "Transform (Modify) Bone - Bone: Party_Gold_backpack_clavicle_l"!(ComponentPose=@n6, Alpha=1, Translation="0,0,0", Rotation="0,0,0") -> (Pose) {
    rotationMode=BMM_Replace
    rotationSpace=BCS_WorldSpace
    translationMode=BMM_Replace
    translationSpace=BCS_WorldSpace
    alphaBoolBlend {blendOption=Linear}
    boneToModify {boneName=Party_Gold_backpack_clavicle_l}
   }
   n8 = "Transform (Modify) Bone - Bone: Party_Gold_backpack_upperarm_l"!(ComponentPose=@n7, Alpha=1, Translation="0,0,0", Rotation="0,0,0") -> (Pose) {
    rotationMode=BMM_Replace
    rotationSpace=BCS_WorldSpace
    translationMode=BMM_Replace
    translationSpace=BCS_WorldSpace
    alphaBoolBlend {blendOption=Linear}
    boneToModify {boneName=Party_Gold_backpack_upperarm_l}
   }
   Save cached pose 'Start Pose'(Pose=Component To Local(ComponentPose=@n8)) {cacheName=Start Pose}
  ]}
 EventGraph {nodes [
   Event BlueprintUpdateAnimation()
   > n1 = Gravity Override State Function DEPRECATED(CharacterPartAnimInstance=Self-Reference(), GameTimeInSeconds=GetGameTimeInSeconds(), GravityOverrideParams=Get Gravity Override Params())
   > Set Graviti Override(Graviti Override:Vector=@n1) -> (Output_Get:Vector)
   n2 = Get BodyMesh() -> (BodyMesh:Skeletal Mesh Component Object Reference)
   n3 = BreakTransform(InTransform=GetSocketTransform(self=@n2, InSocketName=clavicle_l, TransformSpace=RTS_World))
   > Set Clavicle Translation(Clavicle Translation:Vector=@n3) -> (Output_Get:Vector)
   > Set Clavicle Rotation(Clavicle Rotation:Rotator=@n3.Rotation) -> (Output_Get:Rotator)
   n4 = BreakTransform(InTransform=GetSocketTransform(self=@n2, InSocketName=upperarm_l, TransformSpace=RTS_World))
   > Set UpperArm Translation(UpperArm Translation:Vector=@n4) -> (Output_Get:Vector)
   > Set UpperArm Rotation(UpperArm Rotation:Rotator=@n4.Rotation) -> (Output_Get:Rotator)
   n5 = Event OnCharacterCustomizationLoaded()
   > Set BodyMesh(BodyMesh=Get Array Item with EFort Custom Part Type(Array=@n5.CustomMeshParts))
  ]}
]
variables [
 BodyMesh {type=SkeletalMeshComponent}
 Clavicle Rotation {type=Rotator}
 Clavicle Translation {type=Vector}
 Graviti Override {type=Vector}
 Gravity Override Params {type=GravityOverrideParamsStruct}
 UpperArm Rotation {type=Rotator}
 UpperArm Translation {type=Vector}
]
